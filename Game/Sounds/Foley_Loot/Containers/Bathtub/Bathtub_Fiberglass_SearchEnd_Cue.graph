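# Sound cue for the foley that plays when a fiberglass bathtub loot search ends.
# A quality switch sends its first two levels to a mix of the no-more-loot UI sound and a bathtub wave, and the other two to a single "opt" bathtub wave.
# Both bathtub waves get slight random pitch and volume variation.
# Prop-search attenuation and a concurrency cap that stops the farthest instance limit playback.
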SoundCue /Game/Sounds/Foley_Loot/Containers/Bathtub/Bathtub_Fiberglass_SearchEnd_Cue
root=QualityLevel_0
nodes [
 {id=Mixer_13, children [WavePlayer_44, Modulator_16], properties {inputVolume [0.8,1]}}
 {id=Modulator_0, children [WavePlayer_1], properties {pitchMax=1.07, pitchMin=0.92, volumeMax=1.07, volumeMin=0.92}}
 {id=Modulator_16, children [WavePlayer_67], properties {pitchMax=1.07, pitchMin=0.92, volumeMax=1.07, volumeMin=0.92}}
 {id=QualityLevel_0, children [Mixer_13,Mixer_13,Modulator_0,Modulator_0]}
 {id=WavePlayer_1, properties {soundWaveAssetPtr=&/Bathtub_Fiberglass_SearchEnd_opt}}
 {id=WavePlayer_44, properties {soundWaveAssetPtr=/Game/Sounds/Fort_UI_Sounds/Gameplay/Fort_UI_NoMoreLoot}}
 {id=WavePlayer_67, properties {soundWaveAssetPtr=&/Bathtub_Fiberglass_SearchEnd2}}
]
properties {
 attenuationSettings=/Game/Sounds/Attenuation/Fort_Prop_Search_Att
 bOverrideConcurrency=true
 maxDistance=1300
 soundClassObject=&Fort_Loot_Collect
 volumeMultiplier=0.5
 concurrencyOverrides {maxCount=4, resolutionRule=StopFarthestThenPreventNew}
}
