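# Sound cue for searching a fiberglass bathtub loot container, using prop-search attenuation and the loot-collect sound class.
# A quality switch feeds the first two levels with mixes of randomized, pitch- and volume-varied bathtub waves under louder random gear-rummage waves.
# The second level has fewer variations, and the last two each play a single optimized wave.
# Its concurrency override prevents new instances once the limit is reached.
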SoundCue /Game/Sounds/Foley_Loot/Containers/Bathtub/Bathtub_Fiberglass_Search_Cue
root=QualityLevel_0
nodes [
 {id=Mixer_0, children [Modulator_2,Random_1], properties {inputVolume [0.3,1.2]}}
 {id=Mixer_12, children [Modulator_22,Random_11], properties {inputVolume [0.3,1.2]}}
 {id=Modulator_0, children [WavePlayer_0]}
 {id=Modulator_2, children [Random_0], properties {pitchMax=1.07, pitchMin=0.92, volumeMax=1.07, volumeMin=0.92}}
 {id=Modulator_22, children [Random_13], properties {pitchMax=1.07, pitchMin=0.92, volumeMax=1.07, volumeMin=0.92}}
 {id=Modulator_3, children [WavePlayer_1]}
 {id=QualityLevel_0, children [Mixer_12,Mixer_0,Modulator_0,Modulator_3]}
 {id=Random_0, children [WavePlayer_2,WavePlayer_3], properties {weights [1,1]}}
 {id=Random_1, children [WavePlayer_4,WavePlayer_5], properties {bRandomizeWithoutReplacement=false, weights [1,1]}}
 {id=Random_11, children [WavePlayer_52, WavePlayer_54, WavePlayer_56], properties {weights [1,1,1]}}
 {id=Random_13, children [WavePlayer_68, WavePlayer_69, WavePlayer_70], properties {weights [1,1,1]}}
 {id=WavePlayer_0, properties {soundWaveAssetPtr=&/Bathtub_Fiberglass_Search_opt}}
 {id=WavePlayer_1, properties {soundWaveAssetPtr=&Bathtub_Fiberglass_Search_opt}}
 {id=WavePlayer_2, properties {soundWaveAssetPtr=&/Bathtub_Fiberglass_Search03}}
 {id=WavePlayer_3, properties {soundWaveAssetPtr=&/Bathtub_Fiberglass_Search04}}
 {id=WavePlayer_4, properties {soundWaveAssetPtr=/Game/Sounds/Foley_Loot/Containers/Misc/Prop_Gear_Search_05}}
 {id=WavePlayer_5, properties {soundWaveAssetPtr=/Game/Sounds/Foley_Loot/Containers/Misc/Prop_Gear_Search_06}}
 {id=WavePlayer_52, properties {soundWaveAssetPtr=&Prop_Gear_Search_05}}
 {id=WavePlayer_54, properties {soundWaveAssetPtr=&Prop_Gear_Search_06}}
 {id=WavePlayer_56, properties {soundWaveAssetPtr=/Game/Sounds/Foley_Loot/Containers/Misc/Prop_Gear_Search_01}}
 {id=WavePlayer_68, properties {soundWaveAssetPtr=&Bathtub_Fiberglass_Search03}}
 {id=WavePlayer_69, properties {soundWaveAssetPtr=&Bathtub_Fiberglass_Search04}}
 {id=WavePlayer_70, properties {soundWaveAssetPtr=&/Bathtub_Fiberglass_Search05}}
]
properties {
 attenuationSettings=/Game/Sounds/Attenuation/Fort_Prop_Search_Att
 bOverrideConcurrency=true
 maxDistance=1300
 soundClassObject=&Fort_Loot_Collect
 volumeMultiplier=0.95
 concurrencyOverrides {resolutionRule=PreventNew}
}
